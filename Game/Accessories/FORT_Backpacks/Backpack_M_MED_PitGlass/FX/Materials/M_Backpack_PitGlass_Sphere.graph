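# Unlit Niagara mesh-particle material for the PitGlass backpack's glowing energy sphere, with lightning-like patterns.
# Camera-facing virtual-plane sphere UVs, warped by a turbulence flowmap, drive three panned noise samples screen-blended with a center glow mask.
# Fresnel gradients, perturbed by the same flowmap, color the lightning, body and center glow; low/medium quality skips the flow distortion.
# An Overlay switch substitutes particle-colored directional highlights; a dynamic parameter scales emissive.
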Material /Game/Accessories/FORT_Backpacks/Backpack_M_MED_PitGlass/FX/Materials/M_Backpack_PitGlass_Sphere
graph {
 properties {bUsedWithNiagaraMeshParticles=true, pixelDepthOffsetMode=PDOM_Legacy, shadingModel=MSM_Unlit, usedShadingModels=Unlit}
 nodes [
  n1 = Subtract(A=CameraPositionWS(), B=WorldPosition().XYZ)
  Add(
   A=Multiply(A=Normalize(VectorInput=@n1), B=ScalarParameter() {parameterName=Camera Offset})
   B=Multiply(A=VertexNormalWS(), B=ScalarParameter() {defaultValue=0.333, parameterName=Normal Offset})
  )
  n2 = Divide(A=Saturate(None=DynamicParameter() {paramNames [Emissive Global Scalar, Param2, Param3, Param4]}.Emissive Global Scalar))
  n3 = VertexInterpolator(VS=PreSkinnedNormal())
  n4 = Constant3Vector() {constant=<rgba 1 0 0 0>}
  n5 = Power(
   Base=Multiply(A=Sine(None=@n2), B=DotProduct(A=@n3.PS, B=@n4))
   Exp=ScalarParameter() {defaultValue=1, parameterName=Lightning Highlight Exponent}
  )
  n6 = NamedRerouteUsage()
  n7 = CameraVectorWS()
  n8 = CrossProduct(A=@n7, B=Transform(None=Constant3Vector() {constant=<rgba 0 1 0 0>}) {transformSourceType=TRANSFORMSOURCE_View})
  n9 = MaterialFunctionCall!(
   UV=MaterialFunctionCall(
    Plane Normal Axis=@n7
    Plane X Axis=Normalize(VectorInput=@n8)
    Plane Center=MaterialFunctionCall() {materialFunction=&ComponentPivotLocation}.Result
    UV Scale=ScalarParameter() {defaultValue=17, parameterName=UV Scale}
   ) {materialFunction=&VirtualPlaneCoordinates}.UVs
   Size
   Curvature
  ) {materialFunction=&/MF_Sphere}
  n10 = ScalarParameter() {defaultValue=0.5, parameterName=Flowmap UV Scale, sliderMax=1}
  n11 = Constant()
  n12 = Multiply(A=ScalarParameter() {defaultValue=0.1, parameterName=Flowmap Panner Speed}, B=TwoSidedSign())
  n13 = TextureSample(UVs=Panner(Coordinate=Multiply(A=@n9.UV, B=@n10), Speed=AppendVector(A=@n11, B=@n12)) {bFractionalPart=true}) {samplerType=SAMPLERTYPE_LinearColor, texture=/Game/Effects/Fort_Effects/Textures/Plasma/T_Turbulence_AdditiveWarp_Linear}
  n14 = Abs(None=DotProduct(A=PixelNormalWS(), B=CameraVectorWS()))
  n15 = Multiply(A=ConstantBiasScale(None=ComponentMask(None=@n13.RGB) {g=true, r=true}) {bias=-0.5, scale=2}, B=SquareRoot(None=@n14))
  n16 = Add(A=@n9.UV, B=Multiply(A=ScalarParameter() {defaultValue=0.066, parameterName=UV Flowmap Intensity}, B=@n15))
  n17 = Transform(None=VertexNormalWS()) {transformSourceType=TRANSFORMSOURCE_World, transformType=TRANSFORM_Camera}
  n18 = VertexInterpolator(VS=ComponentMask(None=ConstantBiasScale(None=@n17)) {g=true, r=true})
  n19 = StaticBool() {value=true}
  n20 = MaterialFunctionCall!(
   UV=Divide(A=QualitySwitch(Default=@n16, Low=@n18.PS, Medium=@n18.PS), B=ScalarParameter() {defaultValue=2.618, parameterName=Caustics UV Scale})
   Rotation=Constant() {r=0.25}
   Speed=ScalarParameter() {defaultValue=0.066, parameterName=Caustics Speed}
   Time
   Rotate UVs=@n19
   Fractional Part=@n19
  ) {materialFunction=&/MF_3UVPanner}
  n21 = TextureObject() {
   samplerType=SAMPLERTYPE_LinearGrayscale
   texture=/Game/Accessories/FORT_Backpacks/Backpack_M_MED_PitGlass/FX/Textures/T_Backpack_PitGlass_Noise
  }
  n22 = ScalarParameter() {defaultValue=1, parameterName=Texture Exponent}
  n23 = StaticSwitchParameter(True=Multiply(A=OneMinus(None=@n5), B=@n22), False=@n22) {parameterName=Overlay Material}
  n24 = Add(
   A=Power(Base=TextureSample(UVs=@n20.UV1, Tex=@n21) {texture=/Game/Effects/Fort_Effects/Textures/Tile/T_ShieldCaustics_Tile_Chromatic}.R, Exp=@n23)
   B=Power(Base=TextureSample(UVs=@n20.UV2, Tex=@n21) {texture=&T_ShieldCaustics_Tile_Chromatic}.R, Exp=@n23)
  )
  n25 = Power(Base=TextureSample(UVs=@n20.UV3, Tex=@n21) {texture=&T_ShieldCaustics_Tile_Chromatic}.R, Exp=@n23)
  n26 = MaterialFunctionCall!(Blend=@n6, Base=Divide(A=Add(A=@n24, B=@n25)) {constB=3}) {materialFunction=/Engine/Functions/Engine_MaterialFunctions03/Blends/Blend_Screen}
  n27 = ScalarParameter() {defaultValue=3, parameterName=Blended Texture Exponent}
  n28 = ScalarParameter() {defaultValue=0.333, parameterName=Blended Texture Exponent Highlight}
  n29 = Multiply(A=ParticleColor().RGB, B=Power(Base=Multiply(A=@n5, B=@n26.Result), Exp=LinearInterpolate(A=@n27, B=@n28, Alpha=@n5)))
  n30 = VectorParameter() {parameterName=Color Lightning Fresnel, defaultValue=<rgba 1.3 0 3 0.1>}
  n31 = LinearInterpolate(
   A=VectorParameter() {parameterName=Color Lightning Falloff, defaultValue=<rgba 0.015 0 0.9 3>}.RGB
   B=VectorParameter() {parameterName=Color Lightning, defaultValue=<rgba 0.07 0.845 1 0>}.RGB
   Alpha=MaterialFunctionCall(In=@n26.Result, Power=@n27) {materialFunction=&SCurve}.Result
  )
  n32 = ScalarParameter() {defaultValue=0.5, parameterName=Fresnel Flowmap Intensity Max}
  n33 = ScalarParameter() {defaultValue=0.1, parameterName=Fresnel Flowmap Intensity Min}
  n34 = AppendVector(A=Multiply(A=@n15, B=LinearInterpolate(A=@n32, B=@n33, Alpha=@n14)), B=Constant())
  n35 = Normalize(VectorInput=Add(A=Transform(None=@n34) {transformSourceType=TRANSFORMSOURCE_View}, B=PixelNormalWS()))
  n36 = VertexNormalWS()
  n37 = Abs(None=DotProduct(A=QualitySwitch(Default=@n35, Low=@n36, Medium=@n36) {desc=Skipping flowmap on med/low}, B=CameraVectorWS()))
  n38 = Power(Base=@n37, Exp=@n30.A)
  n39 = VectorParameter() {parameterName=Color C, defaultValue=<rgba 0.0008 0 0.01 2.5>}
  n40 = VectorParameter() {parameterName=Color A, defaultValue=<rgba 0.048 0.816 1.2 1>}
  n41 = LinearInterpolate(A=VectorParameter() {parameterName=Color B, defaultValue=<rgba 0.1332 0.6037 0.666 1>}.RGB, B=@n40.RGB, Alpha=Power(Base=@n37, Exp=@n40.A))
  n42 = Power(Base=OneMinus(None=@n37), Exp=@n39.A)
  n43 = VectorParameter() {parameterName=Center Glow Color A, defaultValue=<rgba 0.05 0.2075 0.5 8>}
  n44 = VectorParameter() {parameterName=Center Glow Color B, defaultValue=<rgba 0 6.6 33 8>}
  n45 = NamedRerouteUsage()
  n46 = Power(Base=@n45, Exp=@n43.A)
  n47 = Add(
   A=Multiply(A=Power(Base=@n26.Result, Exp=@n27), B=LinearInterpolate(A=@n30.RGB, B=@n31, Alpha=@n38))
   B=LinearInterpolate(A=LinearInterpolate(A=@n39.RGB, B=@n41, Alpha=@n42), B=LinearInterpolate(A=@n43.RGB, B=@n44.RGB, Alpha=@n46), Alpha=@n45) {constB=0}
  ) {constB=0}
  MaterialFunctionCall(Emissive=Multiply(
    A=StaticSwitchParameter(True=@n29, False=@n47) {parameterName=Overlay Material}
    B=DynamicParameter() {paramNames [Emissive Global Scalar, Param2, Param3, Param4]}.Emissive Global Scalar
   )) {materialFunction=&MF_FX_EmissiveAdjustments}
  Multiply(A=OneMinus(None=@n6), B=@n27)
  n48 = ScalarParameter() {defaultValue=0.75, parameterName=Center Glow Size}
  n49 = Saturate(None=Divide(A=OneMinus(None=@n37), B=Saturate(None=@n48)))
  NamedRerouteDeclaration(None=Power(Base=OneMinus(None=@n49), Exp=ScalarParameter() {defaultValue=22, parameterName=Center Glow Exponent})) {name=Center Glow Mask, nodeColor=<rgba 1 0.8203 0 1>}
  ScalarParameter() {defaultValue=1, parameterName=Emissive Global Scalar, sliderMax=10000}
  n50 = TwoSidedSign()
  StaticSwitchParameter(True=OneMinus(None=Saturate(None=@n50)), False=Constant() {r=1}) {parameterName=Aura Material}
 ]
}
