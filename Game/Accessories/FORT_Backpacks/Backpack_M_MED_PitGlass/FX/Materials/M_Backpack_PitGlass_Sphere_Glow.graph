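# Additive, unlit, two-sided glow material for the PitGlass backpack's sphere FX, usable on Niagara mesh particles.
# Particle color feeds emissive through a shared FX emissive adjustment function.
# Opacity multiplies particle alpha by two view-angle falloffs, facing and inverted, each with its own exponent parameter, to shape the glow.
# Depth fade softens intersections, but Low and Medium quality and the mobile shading path skip it.
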Material /Game/Accessories/FORT_Backpacks/Backpack_M_MED_PitGlass/FX/Materials/M_Backpack_PitGlass_Sphere_Glow
graph {
 properties {
  bUsedWithNiagaraMeshParticles=true
  blendMode=BLEND_Additive
  pixelDepthOffsetMode=PDOM_Legacy
  shadingModel=MSM_Unlit
  twoSided=true
  usedShadingModels=Unlit
 }
 nodes [
  n1 = ParticleColor()
  MaterialFunctionCall(Emissive=@n1.RGB) {materialFunction=&MF_FX_EmissiveAdjustments}
  n2 = Abs(None=DotProduct(A=PixelNormalWS(), B=CameraVectorWS()))
  n3 = ScalarParameter() {defaultValue=8, parameterName=Fresnel Exponent}
  n4 = OneMinus(None=@n2)
  n5 = ScalarParameter() {parameterName=Center Fresnel Exponent}
  n6 = Multiply(A=@n1.A, B=Multiply(A=Power(Base=@n2, Exp=@n3), B=Power(Base=@n4, Exp=@n5)))
  n7 = ScalarParameter() {defaultValue=11, parameterName=Depth Fade}
  QualitySwitch(
   Default=ShadingPathSwitch(Default=DepthFade(Opacity=@n6, FadeDistance=@n7), Mobile=@n6) {desc=Migrated from FeatureLevelSwitch}
   Low=@n6
   Medium=@n6
  )
  VertexNormalWS()
 ]
}
